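# Mono MetaSound source for the TidalCrane Cluster Launcher laser loop, using first-person weapon sound class and foley attenuation settings.
# On play, a Wave Player starts the laser wave, exposed as the Sound input, and a trigger route selects full gain.
# A delayed trigger later switches the route to a lower gain.
# InterpTo smooths that value before it multiplies the audio, so the laser settles quieter after its initial emphasis.
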MetaSoundSource /ClusterLauncher/Assets/Audio/MSS_TidalCrane_ClusterLauncher_Laser_Loop_Preset
inputs [
 Sound {type=WaveAsset}
 UE.Source.OnPlay {type=Trigger}
]
interfaces [UE.OutputFormat.Mono:1.0, UE.Source:1.1]
outputs ["UE.OutputFormat.Mono.Audio:0" {type=Audio}]
pages [Default {
  input_defaults {Sound=/ClusterLauncher/Assets/Audio/Wavs/TidalCrane_ClusterLauncher_Laser_Loop_A, UE.Source.OnPlay=null}
  nodes [
   TriggerDelay_0 = UE.Trigger Delay(In=^UE.Source.OnPlay) {also {defaults {Delay Time=2}}}
   TriggerRouteTriggerRouteFloat2_0 = "TriggerRoute.Trigger Route (Float, 2):Float"(Set 1=^UE.Source.OnPlay, Set 0=@TriggerDelay_0.Out) {also {defaults {Value 0=0.4, Value 1=1}}}
   WavePlayer_0 = UE.Wave Player:Mono(Wave Asset=^Sound, Play=^UE.Source.OnPlay) {also {defaults {Loop="<MetaSoundBuilderResult.SUCCEEDED: 0>"}, version [1,1]}}
   Multiply_0 = UE.Multiply:Audio by Float(
    AdditionalOperands=UE.InterpTo:Audio(Target=@TriggerRouteTriggerRouteFloat2_0.Value) {also {defaults {Interp Time=2}}}.Value
    PrimaryOperand=@WavePlayer_0.Out Mono
   )
  ]
  links [{dst="@UE.OutputFormat.Mono.Audio:0", src=Multiply_0.Out}]
 }]
properties {
 assetImportData=&self:AssetImportData
 attenuationSettings=/FortMetaSoundWeapon/Items/Shared/Settings/FPS/AS_Weapon_FP_Foley
 bLooping=true
 duration=10000
 maxDistance=2250
 sampleRate=48000
 soundClassObject=/Game/Sounds/FortSoundClassesAndModes/Fort_Events_Weapons_1P
 volume=0.02
}
