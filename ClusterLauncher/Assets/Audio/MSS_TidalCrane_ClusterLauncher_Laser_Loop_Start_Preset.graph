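# One-shot mono MetaSound for the start of the TidalCrane ClusterLauncher laser loop, routed to the first-person weapon sound class with foley attenuation.
# On play, it picks one of three wave variations at random and feeds it to a wave player with a random pitch shift around a base value.
# The wave player's completion fires OnFinished.
# The random volume inputs and the Value output are left unconnected.
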MetaSoundSource /ClusterLauncher/Assets/Audio/MSS_TidalCrane_ClusterLauncher_Laser_Loop_Start_Preset
inputs [
 Random.Pitch.Base {type=Float}
 Random.Pitch.Spread {type=Float}
 Random.Volume.Base {type=Float}
 Random.Volume.Spread {type=Float}
 Sounds {type=WaveAsset:Array}
 UE.Source.OnPlay {type=Trigger}
]
interfaces [UE.OutputFormat.Mono:1.0, UE.Source.OneShot:1.0, UE.Source:1.1]
outputs [
 "UE.OutputFormat.Mono.Audio:0" {type=Audio}
 UE.Source.OneShot.OnFinished {type=Trigger}
 Value {type=Float}
]
pages [Default {
  input_defaults {
   Random.Pitch.Base=0
   Random.Pitch.Spread=0.1
   Random.Volume.Base=0
   Random.Volume.Spread=0
   UE.Source.OnPlay=null
   Sounds [
    /ClusterLauncher/Assets/Audio/Wavs/TidalCrane_ClusterLauncher_Laser_Loop_Start_01
    /ClusterLauncher/Assets/Audio/Wavs/TidalCrane_ClusterLauncher_Laser_Loop_Start_02
    /ClusterLauncher/Assets/Audio/Wavs/TidalCrane_ClusterLauncher_Laser_Loop_Start_03
   ]
  }
  nodes [
   ArrayRandomGet_0 = Array.Random Get:WaveAsset:Array(In Array=^Sounds, Next=^UE.Source.OnPlay) {also {defaults {Weights []}}}
   WavePlayer_0 = UE.Wave Player:Mono(
    Play=@ArrayRandomGet_0.On Next
    Wave Asset=@ArrayRandomGet_0.Value
    Pitch Shift=UE.RandomFloat(
     Next=^UE.Source.OnPlay
     Max=UE.Add:Float(PrimaryOperand=^Random.Pitch.Base, AdditionalOperands=^Random.Pitch.Spread).Out
     Min=UE.Subtract:Float(PrimaryOperand=^Random.Pitch.Base, AdditionalOperands=^Random.Pitch.Spread).Out
    ).Value
   ) {also {version [1,1]}}
  ]
  links [{dst="@UE.Source.OneShot.OnFinished", src=WavePlayer_0.On Finished}, {dst="@UE.OutputFormat.Mono.Audio:0", src=WavePlayer_0.Out Mono}]
 }]
properties {
 assetImportData=&self:AssetImportData
 attenuationSettings=/FortMetaSoundWeapon/Items/Shared/Settings/FPS/AS_Weapon_FP_Foley
 maxDistance=2250
 sampleRate=48000
 soundClassObject=/Game/Sounds/FortSoundClassesAndModes/Fort_Events_Weapons_1P
 volume=0.15
}
