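# Looping stereo MetaSound source for the TidalCrane Cluster Launcher's projectile.
# Stereo and mono wave players start at random offsets, each scaled by its own distance curve from a shared attenuation curve table.
# The boosted stereo layer fades in slowly before both layers are mixed.
# The mix gets a shorter overall fade-in and an IsEnemy-selected gain that makes non-enemy projectiles quieter.
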MetaSoundSource /ClusterLauncher/Assets/Audio/MSS_TidalCrane_ClusterLauncher_Projectile_Loop
inputs [
 FalloffDistance {type=Float}
 IsEnemy {type=Int32}
 UE.Attenuation.Distance {type=Float}
 UE.Source.OnPlay {type=Trigger}
]
interfaces [UE.Attenuation:1.0, UE.OutputFormat.Stereo:1.0, UE.Source:1.1]
outputs [
 "UE.OutputFormat.Stereo.Audio:0" {type=Audio}
 "UE.OutputFormat.Stereo.Audio:1" {type=Audio}
]
pages [Default {
  input_defaults {FalloffDistance=27500, IsEnemy=1, UE.Attenuation.Distance=0, UE.Source.OnPlay=null}
  nodes [
   ADSREnvelopeADSREnvelope_0 = ADSR Envelope.ADSR Envelope:Float(Trigger Attack=^UE.Source.OnPlay) {also {defaults {Attack Time=4, Sustain Level=1}}}
   DecibelstoLinearGain_0 = UE.Decibels to Linear Gain:Float() {also {defaults {Decibels=12}}}
   n1 = UE.Multiply:Float(
    PrimaryOperand=@ADSREnvelopeADSREnvelope_0.Out Envelope
    AdditionalOperands=UE.Multiply:Float(
     AdditionalOperands=@DecibelstoLinearGain_0.Linear Gain
     PrimaryOperand=&MSP_Attenuation_Volume_Curve(FalloffDistance=^FalloffDistance, Distance=^UE.Attenuation.Distance) {also {defaults {AttenuationCurves=/ClusterLauncher/Assets/Audio/Settings/WTB_ClusterLauncher_Projectile_Distance}}}.VolumeMultiplier
    ).Out
   )
   n2 = VariableAccessor.WaveAsset(Variable=InitVariable.WaveAsset().Variable)
   WavePlayer_1 = UE.Wave Player:Stereo(Play=^UE.Source.OnPlay, Start Time=&MSP_WaveGetRandomStartTime(Sound=@n2.Value).Value, Wave Asset=@n2.Value) {also {defaults {Loop="<MetaSoundBuilderResult.SUCCEEDED: 0>"}, version [1,1]}}
   DecibelstoLinearGain_2 = UE.Decibels to Linear Gain:Float()
   n3 = VariableAccessor.WaveAsset(Variable=InitVariable.WaveAsset().Variable)
   WavePlayer_0 = UE.Wave Player:Mono(Play=^UE.Source.OnPlay, Start Time=&MSP_WaveGetRandomStartTime(Sound=@n3.Value).Value, Wave Asset=@n3.Value) {also {defaults {Loop="<MetaSoundBuilderResult.SUCCEEDED: 0>"}, version [1,1]}}
   AudioMixerAudioMixerStereo2_0 = "AudioMixer.Audio Mixer (Stereo, 2)"(
    In 0 R=UE.Multiply:Audio by Float(AdditionalOperands=@n1.Out, PrimaryOperand=@WavePlayer_1.Out Right).Out
    In 0 L=UE.Multiply:Audio by Float(AdditionalOperands=@n1.Out, PrimaryOperand=@WavePlayer_1.Out Left).Out
    Gain 1=UE.Multiply:Float(
     AdditionalOperands=@DecibelstoLinearGain_2.Linear Gain
     PrimaryOperand=&MSP_Attenuation_Volume_Curve(FalloffDistance=^FalloffDistance, Distance=^UE.Attenuation.Distance) {also {defaults {AttenuationCurves=&WTB_ClusterLauncher_Projectile_Distance, CurveIndexToEvaluate=1}}}.VolumeMultiplier
    ).Out
    In 1 L=@WavePlayer_0.Out Mono
    In 1 R=@WavePlayer_0.Out Mono
   )
   TriggerDelay_0 = UE.Trigger Delay(In=^UE.Source.OnPlay) {also {defaults {Delay Time=0}}}
   ADSREnvelopeADSREnvelope_1 = ADSR Envelope.ADSR Envelope:Float(Trigger Attack=@TriggerDelay_0.Out) {also {defaults {Attack Time=1, Sustain Level=1}}}
   DecibelstoLinearGain_1 = UE.Decibels to Linear Gain:Float() {also {defaults {Decibels=-3}}}
   TriggerSelect2_0 = "UE.Trigger Select (2)"(Index=^IsEnemy, In=^UE.Source.OnPlay)
   TriggerRouteTriggerRouteFloat2_0 = "TriggerRoute.Trigger Route (Float, 2):Float"(Set 0=@TriggerSelect2_0.Out 0, Set 1=@TriggerSelect2_0.Out 1) {also {defaults {Value 0=0.7, Value 1=1}}}
   n4 = UE.Multiply:Float(
    AdditionalOperands=UE.Multiply:Float(AdditionalOperands=@ADSREnvelopeADSREnvelope_1.Out Envelope, PrimaryOperand=@DecibelstoLinearGain_1.Linear Gain).Out
    PrimaryOperand=@TriggerRouteTriggerRouteFloat2_0.Value
   )
   Multiply_1 = UE.Multiply:Audio by Float(PrimaryOperand=@AudioMixerAudioMixerStereo2_0.Out R, AdditionalOperands=@n4.Out)
   Multiply_2 = UE.Multiply:Audio by Float(PrimaryOperand=@AudioMixerAudioMixerStereo2_0.Out L, AdditionalOperands=@n4.Out)
  ]
  links [{dst="@UE.OutputFormat.Stereo.Audio:1", src=Multiply_1.Out}, {dst="@UE.OutputFormat.Stereo.Audio:0", src=Multiply_2.Out}]
 }]
properties {
 assetImportData=&self:AssetImportData
 attenuationSettings=/ClusterLauncher/Assets/Audio/Settings/ATT_ClusterLauncher_Projectile
 bLooping=true
 bOverrideConcurrency=true
 duration=10000
 maxDistance=27500
 numChannels=2
 outputFormat=Stereo
 sampleRate=48000
 soundClassObject=&Fort_WeaponRanged_Projectile_Effects
 volume=0.75
 concurrencyOverrides {bEnableMaxCountPlatformScaling=true, maxCount=3, volumeScaleMode=Distance, platformMaxCount {default=5, perPlatform {mobile=2, switch=2}}}
}
