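# MetaSound Source preset for the TidalCrane ClusterLauncher reload-end "ready" sound, played as a mono one-shot.
# It inherits its graph from MSS_Weapon_Reload_End, overriding inputs to supply a single wave, a reduced master volume and pitch-shift base and spread values.
# Fort weapon foley sound class, attenuation and concurrency settings handle categorization, distance falloff and voice limiting.
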MetaSoundSource /ClusterLauncher/Assets/Audio/MSS_TidalCrane_ClusterLauncher_Reload_End_Ready_Preset
preset_parent=/FortMetasound/Weapons/MSS_Weapon_Reload_End
inputs [
 MasterVolume {type=Float}
 Random.PitchShift.Base {type=Float}
 Random.PitchShift.Spread {type=Float}
 Sounds {type=WaveAsset:Array}
 UE.Source.OnPlay {type=Trigger}
]
interfaces [UE.OutputFormat.Mono:1.0, UE.Source.OneShot:1.0, UE.Source:1.1]
outputs [
 "UE.OutputFormat.Mono.Audio:0" {type=Audio}
 UE.Source.OneShot.OnFinished {type=Trigger}
]
pages [Default {input_defaults {
   MasterVolume=0.2
   Random.PitchShift.Base=5
   Random.PitchShift.Spread=1
   UE.Source.OnPlay=null
   Sounds [/ClusterLauncher/Assets/Audio/Wavs/TidalCrane_ClusterLauncher_Reload_End_Ready]
  }}]
properties {
 assetImportData=&self:AssetImportData
 attenuationSettings=&AS_Fort_WeaponFoley
 maxDistance=2250
 sampleRate=48000
 soundClassObject=&Fort_WeaponFoley
 concurrencySet [&CS_Fort_WeaponFoley]
}
