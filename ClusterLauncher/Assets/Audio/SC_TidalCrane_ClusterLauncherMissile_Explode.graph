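# Sound cue for the TidalCrane cluster launcher missile explosion, mixing close, medium and distant layers, each with its own attenuation and sound class node.
# Close blends Ostrich and airstrike explosions with randomly delayed debris, medium adds debris, and distant adds a sweetener.
# Modulators vary the main explosions, while quality-level nodes switch between full and reduced random variation pools.
# Weapon explosion concurrency rules and an override cap simultaneous instances.
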SoundCue /ClusterLauncher/Assets/Audio/SC_TidalCrane_ClusterLauncherMissile_Explode
root=Mixer_0
nodes [
 {
  id=Attenuation_0
  children [Mixer_5]
  properties {
   attenuationSettings=/Game/Sounds/Attenuation/Fort_Explosion_Distant_Athena_Att
   attenuationOverrides {falloffDistance=3000, attenuationShapeExtents {x=1000}}
  }
 }
 {
  id=Attenuation_4
  children [Mixer_1]
  properties {attenuationSettings=/Game/Sounds/Attenuation/Fort_Explosion_Close_Athena_Att, attenuationOverrides {falloffDistance=1600}}
 }
 {
  id=Attenuation_5
  children [Mixer_2]
  properties {
   attenuationSettings=/Game/Sounds/Attenuation/Fort_Explosion_Medium_Athena_Att
   attenuationOverrides {falloffDistance=3000, attenuationShapeExtents {x=1000}}
  }
 }
 {id=Delay_0, children [QualityLevel_2], properties {delayMax=0.3}}
 {id=Delay_1, children [QualityLevel_3]}
 {id=Mixer_0, children [Mixer_4,null], properties {inputVolume [1,0.5]}}
 {id=Mixer_1, children [Modulator_0,Modulator_2,Delay_0], properties {inputVolume [1,1.5,1.5]}}
 {id=Mixer_2, children [Modulator_5,Delay_1], properties {inputVolume [2,0.5]}}
 {id=Mixer_4, children [SoundClass_3,SoundClass_2,SoundClass_0], properties {inputVolume [0.7,0.35,0.5]}}
 {id=Mixer_5, children [Modulator_1, QualityLevel_5], properties {inputVolume [1,0.75]}}
 {id=Modulator_0, children [QualityLevel_0], properties {pitchMax=1.2}}
 {id=Modulator_1, children [QualityLevel_6], properties {pitchMax=1.1, pitchMin=0.9}}
 {id=Modulator_2, children [QualityLevel_1], properties {pitchMax=1.1, pitchMin=0.9}}
 {id=Modulator_5, children [QualityLevel_7], properties {pitchMax=1.1, pitchMin=0.9}}
 {id=QualityLevel_0, children [Random_3,Random_3,Random_0,Random_0]}
 {id=QualityLevel_1, children [Random_1,Random_1,Random_2,Random_2]}
 {id=QualityLevel_2, children [Random_6,Random_6,Random_7,Random_7]}
 {id=QualityLevel_3, children [Random_9,Random_9,Random_10,Random_10]}
 {id=QualityLevel_5, children [Random_13,Random_13,Random_14,Random_14]}
 {id=QualityLevel_6, children [Random_5,Random_5,Random_16,Random_16]}
 {id=QualityLevel_7, children [Random_4,Random_4,Random_18,Random_18]}
 {id=Random_0, children [WavePlayer_2,WavePlayer_6,WavePlayer_8], properties {bShouldExcludeFromBranchCulling=true, weights [1,1,1]}}
 {id=Random_1, children [WavePlayer_0, WavePlayer_12, WavePlayer_13, WavePlayer_14], properties {weights [1,1,1,1]}}
 {id=Random_10, children [WavePlayer_17, WavePlayer_29], properties {bShouldExcludeFromBranchCulling=true, weights [1,1]}}
 {id=Random_13, children [WavePlayer_38, WavePlayer_39, WavePlayer_45, WavePlayer_46], properties {weights [1,1,1,1]}}
 {id=Random_14, children [WavePlayer_38, WavePlayer_46], properties {bShouldExcludeFromBranchCulling=true, weights [1,1]}}
 {id=Random_16, children [WavePlayer_21, WavePlayer_24, WavePlayer_26], properties {bShouldExcludeFromBranchCulling=true, weights [1,1,1]}}
 {id=Random_18, children [WavePlayer_9, WavePlayer_18, WavePlayer_20], properties {bShouldExcludeFromBranchCulling=true, weights [1,1,1]}}
 {id=Random_2, children [WavePlayer_12, WavePlayer_14], properties {bShouldExcludeFromBranchCulling=true, weights [1,1]}}
 {id=Random_3, children [WavePlayer_2,WavePlayer_3,WavePlayer_5,WavePlayer_6,WavePlayer_7,WavePlayer_8], properties {weights [1,1,1,1,1,1]}}
 {
  id=Random_4
  children [WavePlayer_9, WavePlayer_10, WavePlayer_11, WavePlayer_18, WavePlayer_19, WavePlayer_20]
  properties {weights [1,1,1,1,1,1]}
 }
 {
  id=Random_5
  children [WavePlayer_21, WavePlayer_22, WavePlayer_23, WavePlayer_24, WavePlayer_25, WavePlayer_26]
  properties {weights [1,1,1,1,1,1]}
 }
 {id=Random_6, children [WavePlayer_1, WavePlayer_4, WavePlayer_15, WavePlayer_16], properties {weights [1,1,1,1]}}
 {id=Random_7, children [WavePlayer_1, WavePlayer_16], properties {bShouldExcludeFromBranchCulling=true, weights [1,1]}}
 {id=Random_9, children [WavePlayer_17, WavePlayer_27, WavePlayer_28, WavePlayer_29], properties {weights [1,1,1,1]}}
 {
  id=SoundClass_0
  children [Attenuation_0]
  properties {soundClassOverride=&Fort_WeaponRanged_Fire_Distant}
 }
 {
  id=SoundClass_2
  children [Attenuation_5]
  properties {soundClassOverride=&Fort_WeaponRanged_Fire_Distant}
 }
 {
  id=SoundClass_3
  children [Attenuation_4]
  properties {soundClassOverride=&Fort_WeaponRanged_Fire_3P_PSM}
 }
 {id=WavePlayer_0, properties {soundWaveAssetPtr=/Game/Athena/Sounds/Weapons/Air_Strike/AppleSauce_Expl_Close_04}}
 {id=WavePlayer_1, properties {soundWaveAssetPtr=/ClusterLauncher/Assets/Audio/wavs/TidalCrane_ClusterLauncher_Impact_Debris_01}}
 {id=WavePlayer_10, properties {soundWaveAssetPtr=/Game/Athena/Sounds/Vehicles/Ostrich/Veh_Ostrich_Launcher_Expl_Med_02}}
 {id=WavePlayer_11, properties {soundWaveAssetPtr=/Game/Athena/Sounds/Vehicles/Ostrich/Veh_Ostrich_Launcher_Expl_Med_03}}
 {id=WavePlayer_12, properties {soundWaveAssetPtr=/Game/Athena/Sounds/Weapons/Air_Strike/AppleSauce_Expl_Close_01}}
 {id=WavePlayer_13, properties {soundWaveAssetPtr=/Game/Athena/Sounds/Weapons/Air_Strike/AppleSauce_Expl_Close_02}}
 {id=WavePlayer_14, properties {soundWaveAssetPtr=/Game/Athena/Sounds/Weapons/Air_Strike/AppleSauce_Expl_Close_03}}
 {id=WavePlayer_15, properties {soundWaveAssetPtr=/ClusterLauncher/Assets/Audio/wavs/TidalCrane_ClusterLauncher_Impact_Debris_03}}
 {id=WavePlayer_16, properties {soundWaveAssetPtr=/ClusterLauncher/Assets/Audio/wavs/TidalCrane_ClusterLauncher_Impact_Debris_04}}
 {id=WavePlayer_17, properties {soundWaveAssetPtr=&TidalCrane_ClusterLauncher_Impact_Debris_01}}
 {id=WavePlayer_18, properties {soundWaveAssetPtr=/Game/Athena/Sounds/Vehicles/Ostrich/Veh_Ostrich_Launcher_Expl_Med_04}}
 {id=WavePlayer_19, properties {soundWaveAssetPtr=/Game/Athena/Sounds/Vehicles/Ostrich/Veh_Ostrich_Launcher_Expl_Med_05}}
 {id=WavePlayer_2, properties {soundWaveAssetPtr=/Game/Athena/Sounds/Vehicles/Ostrich/Veh_Ostrich_Launcher_Expl_Close_01}}
 {id=WavePlayer_20, properties {soundWaveAssetPtr=/Game/Athena/Sounds/Vehicles/Ostrich/Veh_Ostrich_Launcher_Expl_Med_06}}
 {id=WavePlayer_21, properties {soundWaveAssetPtr=/Game/Athena/Sounds/Vehicles/Ostrich/Veh_Ostrich_Launcher_Expl_Dist_01}}
 {id=WavePlayer_22, properties {soundWaveAssetPtr=/Game/Athena/Sounds/Vehicles/Ostrich/Veh_Ostrich_Launcher_Expl_Dist_02}}
 {id=WavePlayer_23, properties {soundWaveAssetPtr=/Game/Athena/Sounds/Vehicles/Ostrich/Veh_Ostrich_Launcher_Expl_Dist_03}}
 {id=WavePlayer_24, properties {soundWaveAssetPtr=/Game/Athena/Sounds/Vehicles/Ostrich/Veh_Ostrich_Launcher_Expl_Dist_04}}
 {id=WavePlayer_25, properties {soundWaveAssetPtr=/Game/Athena/Sounds/Vehicles/Ostrich/Veh_Ostrich_Launcher_Expl_Dist_05}}
 {id=WavePlayer_26, properties {soundWaveAssetPtr=/Game/Athena/Sounds/Vehicles/Ostrich/Veh_Ostrich_Launcher_Expl_Dist_06}}
 {id=WavePlayer_27, properties {soundWaveAssetPtr=/ClusterLauncher/Assets/Audio/wavs/TidalCrane_ClusterLauncher_Impact_Debris_02}}
 {id=WavePlayer_28, properties {soundWaveAssetPtr=&TidalCrane_ClusterLauncher_Impact_Debris_03}}
 {id=WavePlayer_29, properties {soundWaveAssetPtr=&TidalCrane_ClusterLauncher_Impact_Debris_04}}
 {id=WavePlayer_3, properties {soundWaveAssetPtr=/Game/Athena/Sounds/Vehicles/Ostrich/Veh_Ostrich_Launcher_Expl_Close_02}}
 {id=WavePlayer_38, properties {soundWaveAssetPtr=/ClusterLauncher/Assets/Audio/wavs/TidalCrane_ClusterLauncher_Distant_Sweetner_01}}
 {id=WavePlayer_39, properties {soundWaveAssetPtr=/ClusterLauncher/Assets/Audio/wavs/TidalCrane_ClusterLauncher_Distant_Sweetner_02}}
 {id=WavePlayer_4, properties {soundWaveAssetPtr=&TidalCrane_ClusterLauncher_Impact_Debris_02}}
 {id=WavePlayer_45, properties {soundWaveAssetPtr=/ClusterLauncher/Assets/Audio/wavs/TidalCrane_ClusterLauncher_Distant_Sweetner_03}}
 {id=WavePlayer_46, properties {soundWaveAssetPtr=/ClusterLauncher/Assets/Audio/wavs/TidalCrane_ClusterLauncher_Distant_Sweetner_04}}
 {id=WavePlayer_5, properties {soundWaveAssetPtr=/Game/Athena/Sounds/Vehicles/Ostrich/Veh_Ostrich_Launcher_Expl_Close_03}}
 {id=WavePlayer_6, properties {soundWaveAssetPtr=/Game/Athena/Sounds/Vehicles/Ostrich/Veh_Ostrich_Launcher_Expl_Close_04}}
 {id=WavePlayer_7, properties {soundWaveAssetPtr=/Game/Athena/Sounds/Vehicles/Ostrich/Veh_Ostrich_Launcher_Expl_Close_05}}
 {id=WavePlayer_8, properties {soundWaveAssetPtr=/Game/Athena/Sounds/Vehicles/Ostrich/Veh_Ostrich_Launcher_Expl_Close_06}}
 {id=WavePlayer_9, properties {soundWaveAssetPtr=/Game/Athena/Sounds/Vehicles/Ostrich/Veh_Ostrich_Launcher_Expl_Med_01}}
]
properties {
 bOverrideConcurrency=true
 maxDistance=27500
 soundClassObject=&Fort_WeaponRanged_Fire_3P_PSM
 volumeMultiplier=1
 concurrencyOverrides {maxCount=7}
 concurrencySet [/Game/Sounds/Concurrency_Rules/Weapon_Explosion_Conc]
}
