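# Niagara system for the ClusterLauncher projectile's rocket trail, built from three emitters.
# Airglow_Light_Flare burst-spawns an orange glow, streak sprite, light and mesh, tinted by time of day, with distance-remapped sprite size.
# Puffy smoke spawns per distance travelled, inherits slight velocity, and cools from hot orange to fading smoke.
# A rate-spawned ribbon similarly darkens, widens and fades, with both trails animating dynamic material parameters.
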NiagaraSystem /ClusterLauncher/Assets/FX/NS_ClusterLauncher_PrjTrail
dependencies {
 usedDataInterfaces [&NiagaraDataInterfaceColorCurve, &NiagaraDataInterfaceCurve]
 usedDynamicInputs []
 usedModules [&SystemState]
}
emitters {
 Airglow_Light_Flare {
  emitterUpdateScript [
   /SetVariables(Emitter.TODColorMult:LinearColor=&FNTimeOfDayColor(Night Color=<rgba 1 1 1 1>, Morning Color=<rgba 1.25 1.25 1.25 1>, Day Color=<rgba 1.5 1.5 1.5 1>, Evening Color=<rgba 1.25 1.25 1.25 1>)) {isSetParameters=true}
   &EmitterState(
    Inactive Response=::NewEnumerator1
    Life Cycle Mode=::NewEnumerator1
    Loop Behavior=::NewEnumerator1
    Loop Duration Mode=::NewEnumerator1
    UseLoopDelay=false
   )
   &SpawnBurst_Instantaneous(Use Spawn Probability=false, UseLoopCountLimit=false)
   /SetVariables(
    Emitter.LightColor:LinearColor=&ColorFromCurve(
     ColorCurve=DataInterface() {propertyValues {
       AlphaCurve {keys [<at 0.0513 1 RCIM_Cubic>]}
       BlueCurve {keys [<at 0.084 0>,<at 0.2843 0 RCIM_Cubic>]}
       GreenCurve {keys [<at 0.084 0>,<at 0.2843 3.7041 RCIM_Cubic>]}
       RedCurve {keys [<at 0.084 0>,<at 0.2843 25 RCIM_Cubic>]}
      }}
     CurveIndex=$Emitter.Age
    )
    Emitter.LightExponent=1.5
    Emitter.LightRadius:NiagaraFloat=&Lerp_Float(
     A=0
     B=180
     Alpha=&FloatFromCurve(
      FloatCurve=DataInterface() {propertyValues {Curve {keys [{interpMode=RCIM_Cubic, leaveTangent=1.7761, tangentMode=RCTM_Break, time=0.1, value=0}, <at 1 1>]}}}
      CurveIndex=$Emitter.Age
     )
    )
   ) {isSetParameters=true}
  ]
  particleSpawnScript [
   &PartitionParticles()
   &InitializeParticle(
    Color=<rgba 1 0 0 0.5>
    Lifetime=0.17
    Position Offset=<0 1 0>
    Sprite Size Mode=::NewEnumerator3
    Uniform Sprite Size=150
    UsePositionOffset=true
    Write Material Random=true
   )
   &CameraOffset()
  ]
  particleUpdateScript [
   &ParticleState(Kill Particles When Lifetime Has Elapsed=false, Let Infinitely Lived Particles Die When Emitter Deactivates=true)
   &ScaleSpriteSize(Scale Sprite Size Mode=::NewEnumerator0)
   &FN_Sprite_Attributes(
    Camera Offset=50
    Color=&MultiplyLinearColor(
     Linear Color=&MultiplyLinearColor(
      Linear Color=&ColorFromCurve(ColorCurve=DataInterface() {propertyValues {
         AlphaCurve {keys [<at 0.0609 0 RCIM_Cubic>,<at 0.6127 0.3 RCIM_Cubic>]}
         BlueCurve {keys [<at 0.0767 0 RCIM_Cubic>,<at 0.5465 0 RCIM_Cubic>]}
         GreenCurve {keys [<at 0.0767 0.0333 RCIM_Cubic>,<at 0.5465 0.1 RCIM_Cubic>]}
         RedCurve {keys [<at 0.0767 1 RCIM_Cubic>,<at 0.5465 1 RCIM_Cubic>]}
        }})
      Scale Factor=&MakeVector4FromVectorAndFloat("Vector (RGB)"=&VectorFromFloat(Value=5), "Float (Alpha)"=1)
     )
     Scale Factor=&MakeVector4_FromLinearColor(LinearColor=$Emitter.TODColorMult)
    )
    Create Camera Offset Attribute=true
    Create Color Attribute=true
    "Create Normalized Age "=true
    Create SpriteSize Attribute=true
    SpriteSize=&Multiply_Vector2D_ByFloat(
     Float=&Multiply_Float(
      A=&FloatFromCurve(FloatCurve=DataInterface() {propertyValues {Curve {keys [
           {interpMode=RCIM_Cubic, leaveTangent=4.0017, tangentMode=RCTM_Break, time=0, value=0}
           {arriveTangent=0.2116, interpMode=RCIM_Cubic, tangentMode=RCTM_Break, time=0.5, value=1.55}
          ]}}})
      B=&FN_LOD_Distance_Remap(Start Distance=0, End Distance=1000, Minimum Output=0.1, Maximum Output=1, Clamp Output=true)
     )
     Vector2D=<150 150>
    )
    Write to Module Namespace=true
   )
   &FN_Light_Attributes(
    Color=&MultiplyLinearColor(
     Linear Color=<rgba 1 0.3872 0 1>
     Scale Factor=&MakeVector4_FromLinearColor(LinearColor=$Emitter.TODColorMult)
    )
    Create Color Attribute=true
    Create Exponent Attribute=true
    Create Radius Attribute=true
    Enable Particle Light Enable=true
    Exponent=5
    Light Enabled=true
    Radius=500
    Scale Color=5
   )
   &FN_Sprite_Attributes(
    Camera Offset:NiagaraFloat=&Lerp_Float(
     A=0
     B=10
     Alpha=&FN_LOD_Distance_Remap(Start Distance=0, End Distance=10000, Minimum Output=0, Maximum Output=1, Clamp Output=true)
    )
    Color=&MultiplyLinearColor(
     Linear Color=&MultiplyLinearColor(
      Linear Color=&ColorFromCurve(ColorCurve=DataInterface() {propertyValues {
         AlphaCurve {keys [<at 0.0006 0>,<at 1 1 RCIM_Cubic>]}
         BlueCurve {keys [<at 0 0 RCIM_Cubic>]}
         GreenCurve {keys [<at 0 0.1 RCIM_Cubic>]}
         RedCurve {keys [<at 0 1 RCIM_Cubic>]}
        }})
      Scale Factor=&MakeVector4FromVectorAndFloat("Vector (RGB)"=&VectorFromFloat(Value=1), "Float (Alpha)"=0.35)
     )
     Scale Factor=&MakeVector4_FromLinearColor(LinearColor=$Emitter.TODColorMult)
    )
    Create Camera Offset Attribute=true
    Create Color Attribute=true
    Create SpriteSize Attribute=true
    SpriteSize=&Multiply_Vector2D_ByFloat(
     Float=&Lerp_Float(
      A=0
      B=1
      Alpha=&FN_LOD_Distance_Remap(Start Distance=0, End Distance=5000, Minimum Output=0.01, Maximum Output=1, Clamp Output=true)
     )
     Vector2D=&Multiply_Vector2D_ByFloat(
      Float=&RandomRangeFloat(Minimum=1, Maximum=1.5, Evaluation Type=::NewEnumerator1)
      Vector2D=&Multiply_Vector2D_ByFloat(Float=1.3, Vector2D=<300 5>)
     )
    )
    Write to Module Namespace=true
   ) {module=FN_Sprite_Attributes001}
   &FN_Mesh_Attributes(
    Color=&MultiplyLinearColor(
     Linear Color=<rgba 4 1.6088 0 1>
     Scale Factor=&MakeVector4_FromLinearColor(LinearColor=$Emitter.TODColorMult)
    )
    Create Color Attribute=true
    Create Orientation Attribute=true
    Create Scale Attribute=true
    Orientation:Quat4f=&MultiplyQuaternion(
     Quaternion A=<0 0 0 1>
     Quaternion B=&MakeQuaternion(Quaternion From=0, Axis=<0 0 1>, Angle=-0.25, From=?, To=?, Angle Type=::NewEnumerator1, Normalize Vectors=false)
    )
    Scale=&Multiply_VectorByFloat(Vector=<0.2 0.25 0.25>, Float=2)
    Write to Module Namespace=true
   )
  ]
  rendererClasses [
   &NiagaraSpriteRendererProperties
   &NiagaraLightRendererProperties
   &NiagaraMeshRendererProperties
  ]
  renderers [
   {rendererClass=&NiagaraSpriteRendererProperties}
   {rendererClass=&NiagaraLightRendererProperties, rendererIndex=1}
   {rendererClass=&NiagaraSpriteRendererProperties, rendererIndex=2}
   {rendererClass=&NiagaraMeshRendererProperties, rendererIndex=3}
  ]
 }
 Rocket_Trail_Puffy_Smoke {
  emitterUpdateScript [
   &EmitterState(Life Cycle Mode=::NewEnumerator1, Loop Duration=0.5, UseLoopDelay=false)
   &SpawnPerUnit(Module.Use Max Movement Threshold=false, Module.Use Movement Tolerance=false, Spawn Spacing=40, Use Spawn Probability=false)
  ]
  particleSpawnScript [
   &InitializeParticle(
    Color=<rgba 0.1 0.1 0.1 0.2>
    Lifetime Max=1
    Lifetime Min=0.25
    Lifetime Mode=::NewEnumerator1
    Sprite Rotation Mode=::NewEnumerator2
    Sprite Size Mode=::NewEnumerator1
    Sprite UV Mode=::NewEnumerator3
    Uniform Sprite Size Max=180
    Uniform Sprite Size Min=120
    UsePositionOffset=false
    Write Material Random=true
   )
   &InheritVelocity(Inherited Velocity Amount Scale=<0.1 0.1 0.1>, Inherited Velocity Speed Limit=5000, Speed Limit=true, UseSpeedThreshold=false)
  ]
  particleUpdateScript [
   &SolveForcesAndVelocity(Clamp Velocity=false, Limit Acceleration=false)
   &ParticleState()
   &ScaleSpriteSize(Uniform Curve Sprite Scale=DataInterface() {propertyValues {Curve {keys [{interpMode=RCIM_Cubic, leaveTangent=2.3279, tangentMode=RCTM_Break, time=0, value=0}, {tangentMode=RCTM_User, time=1, value=1}]}}})
   &ScaleColor(
    Linear Color Curve=DataInterface() {propertyValues {
      AlphaCurve {keys [<at 0 0.7>,<at 1 0>]}
      BlueCurve {keys [<at 0 0>,<at 0.107 1>]}
      GreenCurve {keys [<at 0 8.4435>,<at 0.107 1>]}
      RedCurve {keys [<at 0 50>,<at 0.107 1>]}
     }}
    Scale Mode=::NewEnumerator2
   )
   &DynamicMaterialParameters(
    Index 0 Param 1=&FloatFromCurve(
     FloatCurve=DataInterface() {propertyValues {Curve {keys [
         {interpMode=RCIM_Cubic, leaveTangent=15.2028, tangentMode=RCTM_Break, time=0, value=1}
         {interpMode=RCIM_Cubic, leaveTangent=10.7187, tangentMode=RCTM_Break, time=1, value=6}
        ]}}}
     Scale Curve=0
    )
    Index 0 Param 3=&FloatFromCurve(FloatCurve=DataInterface() {propertyValues {Curve {keys [{interpMode=RCIM_Cubic, tangentMode=RCTM_Break, time=0, value=1}, <at 1 -1>]}}})
    Index 0 Param 4=1
    Module.Index0Param0Write=true
    Module.Index0Param1Write=true
    Module.Index0Param2Write=true
    Module.Index0Param3Write=true
   )
  ]
  rendererClasses [&NiagaraSpriteRendererProperties]
  renderers [{rendererClass=&NiagaraSpriteRendererProperties}]
 }
 Rocket_Trail_Ribbon {
  emitterUpdateScript [
   &EmitterState()
   &SpawnRate(SpawnRate=20, Use Spawn Probability=false)
  ]
  particleSpawnScript [&InitializeParticle(
    Color=<rgba 0.01 0.01 0.01 0.2>
    Lifetime=0.7
    Ribbon Width=25
    Ribbon Width Mode=::NewEnumerator1
    UsePositionOffset=false
    Write Material Random=true
   )]
  particleUpdateScript [
   &ParticleState()
   &ScaleColor(
    Scale Mode=::NewEnumerator1
    Scale RGBA=&Multiply_Vector4(
     A=&MakeVector4_FromLinearColor(LinearColor=&ColorFromCurve(ColorCurve=DataInterface() {propertyValues {
         AlphaCurve {keys [<at 0.2284 1>,<at 1 0>]}
         BlueCurve {keys [<at 0 0>,<at 0.1023 0.1>]}
         GreenCurve {keys [<at 0 0.9227>,<at 0.1023 0.1>]}
         RedCurve {keys [<at 0 2>,<at 0.1023 0.1>]}
        }}))
     B=&MakeVector4FromVectorAndFloat("Vector (RGB)"=<1 1 1>, "Float (Alpha)"=0.85)
    )
    ScaleRGBA=true
   )
   &ScaleColor(
    Scale Mode=::NewEnumerator1
    Scale RGBA=&Multiply_Vector4(
     A=&MakeVector4_FromLinearColor(LinearColor=&ColorFromCurve(ColorCurve=DataInterface() {propertyValues {
         AlphaCurve {keys [<at 0 1>,<at 1 0>]}
         BlueCurve {keys [<at 0 0>,<at 0.9981 20>]}
         GreenCurve {keys [<at 0 9.2273>,<at 0.9981 20>]}
         RedCurve {keys [<at 0 20>,<at 0.9981 20>]}
        }}))
     B=&MakeVector4FromVectorAndFloat(
      "Vector (RGB)"=<1 1 1>
      "Float (Alpha)"=&FloatFromCurve(FloatCurve=DataInterface() {propertyValues {Curve {keys [<at 0 0>,<at 0.5 1>]}}}, CurveIndex=$Engine.System.Age)
     )
    )
    ScaleRGBA=true
   ) {enabled=false, module=ScaleColor002}
   &DynamicMaterialParameters(
    Index 0 Param 1=&FloatFromCurve(FloatCurve=DataInterface() {propertyValues {Curve {keys [{interpMode=RCIM_Cubic, tangentMode=RCTM_Break, time=0, value=1.35}, <at 1 0>]}}})
    Index 0 Param 3=1
    Index 0 Param 4=1
    Module.Index0Param0Write=true
    Module.Index0Param1Write=true
    Module.Index0Param2Write=true
    Module.Index0Param3Write=true
   )
   &ScaleRibbonWidth(Ribbon Width Scale=&FloatFromCurve(
     FloatCurve=DataInterface() {propertyValues {Curve {keys [{interpMode=RCIM_Cubic, tangentMode=RCTM_Break, time=0, value=1}, <at 1 2>]}}}
     Scale Curve=2
    ))
   &ScaleColor(Scale RGB=&VectorFromFloat(Value=1), ScaleA=true, ScaleRGB=true) {enabled=false, module=ScaleColor001}
  ]
  rendererClasses [&NiagaraRibbonRendererProperties]
  renderers [{rendererClass=&NiagaraRibbonRendererProperties}]
 }
}
summary {
 systemName=NS_ClusterLauncher_PrjTrail
 emitters [
  {emitterName=Rocket_Trail_Ribbon, rendererClasses [&NiagaraRibbonRendererProperties]}
  {emitterName=Rocket_Trail_Puffy_Smoke, rendererClasses [&NiagaraSpriteRendererProperties]}
  {
   emitterName=Airglow_Light_Flare
   rendererClasses [
    &NiagaraSpriteRendererProperties
    &NiagaraLightRendererProperties
    &NiagaraMeshRendererProperties
   ]
  }
 ]
 userVariables [User.ProjLifetime {type=NiagaraFloat, defaultValue=10, description=""}]
}
