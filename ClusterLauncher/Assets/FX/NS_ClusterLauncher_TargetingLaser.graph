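# Niagara system for the Cluster Launcher's red targeting laser, drawn from a sampled socket to User.BeamEnd_0.
# MainRibbon lays partitioned particles along a Hermite spline, shaping width along the beam and alpha by beam length.
# Glow places two particles at the endpoints with sprite flares and a light, scaling and camera-offsetting the far flare by camera distance.
# IsLocalPlayer tunes flare size and material; IsCinematic swaps the target for a fixed-length local-Y offset.
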NiagaraSystem /ClusterLauncher/Assets/FX/NS_ClusterLauncher_TargetingLaser
dependencies {
 usedDataInterfaces [
  &NiagaraDataInterfaceCurve
  &NiagaraDataInterfaceCamera
  &NiagaraDataInterfaceColorCurve
  &NiagaraDataInterfaceSkeletalMesh
  &NiagaraDataInterfaceSocketReader
 ]
 usedDynamicInputs []
 usedModules [
  &SystemState
  &self:SystemSpawnScript.SystemScriptSource.SystemScriptGraph.NiagaraNodeAssignment_8.SetVariables_#1
  &self:SystemSpawnScript.SystemScriptSource.SystemScriptGraph.NiagaraNodeAssignment_0.SetVariables_#2
  &self:SystemSpawnScript.SystemScriptSource.SystemScriptGraph.NiagaraNodeAssignment_1.SetVariables_#3
 ]
}
emitters {
 Glow {
  emitterUpdateScript [
   &EmitterState(Life Cycle Mode=::NewEnumerator1, Loop Behavior=::NewEnumerator1, Loop Duration Mode=::NewEnumerator1, UseLoopDelay=false)
   &SpawnBurst_Instantaneous(Spawn Count=2, Use Spawn Probability=false, UseLoopCountLimit=false)
  ]
  particleSpawnScript [
   &PartitionParticles()
   &InitializeParticle(
    Lifetime=0.1
    Sprite Size Mode=::NewEnumerator3
    Uniform Sprite Size=&MaskFloat(Mask Source=1, Mask Parameter=$Particles.PartitionParticles.Partition, Mask=0, Masked Float=20, Passthrough Float=75)
    UsePositionOffset=false
    Write Material Random=true
   )
  ]
  particleUpdateScript [
   /SetVariables(Particles.DistanceFade:NiagaraFloat=&ClampFloat(Float=&Divide_Float(
      A=&Subtract_Float(
       A=&VectorLength(Vector=&Subtract_Position(
         A=&ReturnCameraPosition(Coordinate Space:ENiagaraCoordinateSpace=::Simulation, Camera Query:NiagaraDataInterfaceCamera=DataInterface())
         B=&TransformPosition(Position=$User.BeamEnd_0, Source Space=::World, Destination Space=::Simulation)
        ))
       B=550
      )
      B=5000
     ))) {isSetParameters=true}
   &CameraOffset(Camera Offset Amount=&MaskFloat(
     Mask Source=1
     Mask Parameter=$Particles.PartitionParticles.Partition
     Mask=0
     Masked Float=0
     Passthrough Float=&Lerp_Float(A=75, B=550, Alpha=$Particles.DistanceFade)
    ))
   &ParticleState(Kill Particles When Lifetime Has Elapsed=false, Let Infinitely Lived Particles Die When Emitter Deactivates=true)
   &ScaleSpriteSize(
    Scale Sprite Size Mode=::NewEnumerator0
    Uniform Scale Factor=&Multiply_Float(
     A=0.7
     B=&MaskFloat(
      Mask Source=1
      Mask Parameter=$Particles.PartitionParticles.Partition
      Mask=0
      Masked Float=1
      Passthrough Float=&Lerp_Float(
       A=&MakeCustomFloatFromBool(Bool=$User.IsLocalPlayer, True Float=1, False Float=0.2)
       B=&MakeCustomFloatFromBool(Bool=$User.IsLocalPlayer, True Float=6, False Float=1)
       Alpha=$Particles.DistanceFade
      )
     )
    )
   )
   &SocketLocation(
    Position Sampling=::NewEnumerator1
    SamplePosOffset=false
    Socket Reader=$System.Socket Reader
    Socket Sampling Mode=::NewEnumerator3
    VelocitySpeedLimit=false
   )
   &ScaleColor(
    Scale RGB=&MakeVectorFromLinearColorRGB(Color=&ColorFromCurve(ColorCurve=DataInterface() {propertyValues {AlphaCurve {keys [<at 0.001 0>,<at 1 1 RCIM_Cubic>]}, BlueCurve {keys [<at 1 0>]}, GreenCurve {keys [<at 1 0>]}, RedCurve {keys [<at 1 1>]}}}))
    ScaleA=true
    ScaleRGB=true
   ) {module=ScaleColor001}
   &ScaleColor(Scale RGB=&VectorFromFloat(Value=20), ScaleA=false, ScaleRGB=true)
   /SetVariables(Particles.Position=/Niagara/DynamicInputs/Helpers/MaskPosition(
     Mask Source:NiagaraInt32=1
     Mask Parameter:NiagaraInt32=$Particles.PartitionParticles.Partition
     Mask:NiagaraInt32=0
     Masked Position:NiagaraPosition=$Output.SocketLocation.SampledPosition
     Passthrough Position:NiagaraPosition=&MakeCustomVectorFromBool(
      Bool=$User.IsCinematic
      True Vector=&AddVectorToPosition(
       Position=$Output.SocketLocation.SampledPosition
       Vector=&Multiply_VectorByFloat(Vector=<0 1 0>, Float=$User.CinematicLength)
      )
      False Vector=&TransformPosition(Position=$User.BeamEnd_0, Source Space=::World, Destination Space=::Local)
     )
    )) {isSetParameters=true}
   /SetVariables(Particles.RoundFlareSize:Vector2f=&Multiply_Vector2D_ByFloat(Float=0.75, Vector2D=$Particles.SpriteSize)) {isSetParameters=true}
   &DynamicMaterialParameters(
    Index 0 Param 1=&MakeCustomFloatFromBool(Bool=$User.IsLocalPlayer, True Float=0.2, False Float=1)
    Index 0 Param 2=&MakeCustomFloatFromBool(Bool=$User.IsLocalPlayer, True Float=5, False Float=0)
    Index 0 Param 3=1
    Index 0 Param 4=1
    Module.Index0Param0Write=true
    Module.Index0Param1Write=true
    Module.Index0Param2Write=true
    Module.Index0Param3Write=true
    Module.Index1Param0Write=true
    Module.Index1Param1Write=true
    Module.Index1Param2Write=true
    Module.Index1Param3Write=true
    Write Parameter Index 1=true
   )
   /SetVariables(Particles.RoundFlareColor:LinearColor=&Multiply_LinearColorByFloat(
     Linear Color=$Particles.Color
     Float=&FloatFromCurve(
      FloatCurve=DataInterface() {propertyValues {Curve {keys [
          {arriveTangent=5.8312, interpMode=RCIM_Cubic, leaveTangent=5.8312, tangentMode=RCTM_User, time=0, value=0}
          {interpMode=RCIM_Cubic, tangentMode=RCTM_User, time=0.5, value=1}
         ]}}}
      Scale Curve=0.25
     )
    )) {isSetParameters=true}
   /SetVariables(Particles.DynMatRoundFlare:Vector4f=<1 0.1 1 0>) {isSetParameters=true}
   &FN_Sprite_Attributes(
    Color=&MultiplyLinearColor(
     Linear Color=$Particles.Color
     Scale Factor=&MakeVector4FromVectorAndFloat("Vector (RGB)"=<1 1 1>, "Float (Alpha)"=0.2)
    )
    Create Color Attribute=true
    Create SpriteSize Attribute=true
    SpriteSize=&Multiply_Vector2D(
     A=$Particles.SpriteSize
     B=/Niagara/DynamicInputs/Helpers/MaskVector2d(
      Mask Source:NiagaraInt32=1
      Mask Parameter:NiagaraInt32=$Particles.PartitionParticles.Partition
      Mask:NiagaraInt32=0
      Masked Vector2D:Vector2f=<0 0>
      Passthrough Vector2D:Vector2f=<8 0.1>
     )
    )
    Write to Module Namespace=true
   )
   &FN_Light_Attributes(
    Color=$Particles.Color
    Create Color Attribute=true
    Create Exponent Attribute=true
    Create Radius Attribute=true
    Enable Particle Light Enable=true
    Exponent=2
    Light Enabled:NiagaraBool=&SetBoolByIntComparison(A=$Particles.PartitionParticles.Partition, B=1, Comparison Type=::NewEnumerator4)
    Radius=75
    Scale Color=2
   )
  ]
  rendererClasses [&NiagaraSpriteRendererProperties, &NiagaraLightRendererProperties]
  renderers [
   {rendererClass=&NiagaraSpriteRendererProperties}
   {rendererClass=&NiagaraSpriteRendererProperties, rendererIndex=1}
   {rendererClass=&NiagaraSpriteRendererProperties, rendererIndex=2}
   {rendererClass=&NiagaraLightRendererProperties, rendererIndex=3}
  ]
 }
 MainRibbon {
  emitterSpawnScript [/Game/Effects/Niagara/Modules/Partition/FN_PartitionAttributes(Number of Bone Sets=1, Particles Per Partition=15)]
  emitterUpdateScript [
   &EmitterState(Life Cycle Mode=::NewEnumerator1, Loop Behavior=::NewEnumerator1, Loop Duration Mode=::NewEnumerator1, UseLoopDelay=false)
   &SpawnBurst_Instantaneous(Spawn Count=$Emitter.FN_PartitionAttributes.Spawn Total, Use Spawn Probability=false)
  ]
  particleSpawnScript [
   &InitializeRibbon(
    Color=<rgba 1 0 0 1>
    Lifetime=0.1
    Module.Write Color=true
    Module.Write Lifetime=true
    Module.Write Mass=false
    Module.Write Material Random=true
    Module.Write Position=false
    Module.Write RibbonTwist=false
    Module.Write RibbonWidth=true
    Ribbon Width=25
   )
   /Game/Effects/Niagara/Modules/Partition/FN_Particle_partition_v5/FN_Particle_Partition_v5(Enable Ribbon ID=true)
  ]
  particleUpdateScript [
   &ParticleState(Kill Particles When Lifetime Has Elapsed=false, Let Infinitely Lived Particles Die When Emitter Deactivates=true)
   &SocketLocation(
    Position Sampling=::NewEnumerator1
    SamplePosOffset=false
    Socket Reader=$System.Socket Reader
    Socket Sampling Mode=::NewEnumerator3
    VelocitySpeedLimit=false
   )
   /SetVariables(Particles.BeamLength:NiagaraFloat=&VectorLength(Vector=&Subtract_Vector(
      A=&TransformPosition(Position=$User.BeamEnd_0, Source Space=::World, Destination Space=::Local)
      B=Linked() {linkedVariable {name=Output.SocketLocation.SampledPosition, type=NiagaraPosition}}
     ))) {isSetParameters=true}
   /Game/Effects/Niagara/Modules/Curves/FN_HermiteCubicSpline_v1/FN_HermiteCubicSpline(
    Position Blend:NiagaraFloat=$Particles.Initial.Partition.NormalizedID
    Position End:Vector3f=&MakeCustomVectorFromBool(
     Bool=$User.IsCinematic
     True Vector=&AddVectorToPosition(
      Position=$Output.SocketLocation.SampledPosition
      Vector=&Multiply_VectorByFloat(Vector=<0 1 0>, Float=$User.CinematicLength)
     )
     False Vector=/Niagara/DynamicInputs/Transforms/TransformPosition(Position=Linked() {linkedVariable {name=User.BeamEnd_0, type=NiagaraPosition}}, Source Space=::World, Destination Space=::Local)
    )
    Position Start:NiagaraPosition=$Output.SocketLocation.SampledPosition
    Write Particles Position=true
   ) {module=FN_HermiteCubicSpline001}
   &ScaleRibbonWidth(Ribbon Width Scale=&FloatFromCurve(
     FloatCurve=DataInterface() {propertyValues {Curve {keys [
         {interpMode=RCIM_Cubic, leaveTangentWeight=0.7456, tangentMode=RCTM_User, time=0, value=1}
         {arriveTangentWeight=0.0667, interpMode=RCIM_Cubic, leaveTangentWeight=0.1667, tangentMode=RCTM_User, time=0.6, value=1}
         {arriveTangentWeight=0.1297, interpMode=RCIM_Cubic, tangentMode=RCTM_User, time=1, value=0.7}
        ]}}}
     CurveIndex=$Particles.Initial.Partition.NormalizedID
     Scale Curve=&Lerp_Float(
      A=0.5
      B=1
      Alpha=&FloatFromCurve(
       FloatCurve=DataInterface() {propertyValues {Curve {keys [
           {interpMode=RCIM_Cubic, tangentMode=RCTM_User, time=0, value=0}
           {arriveTangent=3.4966, interpMode=RCIM_Cubic, leaveTangent=3.4966, tangentMode=RCTM_User, time=0.8, value=1}
           {arriveTangent=-13.2607, interpMode=RCIM_Cubic, leaveTangent=-13.2607, tangentMode=RCTM_User, time=1, value=0}
          ]}}}
       CurveIndex=$Particles.Timeline.NormalizedTime
      )
     )
    ))
   &ScaleColor(
    Scale Alpha=&Lerp_Float(
     A=0.2
     B=&FloatFromCurve(
      FloatCurve=DataInterface() {propertyValues {Curve {keys [
          {arriveTangent=20.8537, interpMode=RCIM_Cubic, leaveTangent=20.8537, leaveTangentWeight=12.2844, tangentMode=RCTM_User, time=0, value=0}
          {arriveTangentWeight=0.1, interpMode=RCIM_Cubic, leaveTangentWeight=0.0667, tangentMode=RCTM_User, time=0.1, value=1}
          {arriveTangentWeight=0.1795, interpMode=RCIM_Cubic, leaveTangentWeight=0.1795, tangentMode=RCTM_User, time=0.2, value=0.8}
          {arriveTangentWeight=0.1491, interpMode=RCIM_Cubic, leaveTangentWeight=0.2236, tangentMode=RCTM_User, time=0.7, value=0.8}
          {arriveTangentWeight=0.2333, interpMode=RCIM_Cubic, leaveTangentWeight=0.3067, tangentMode=RCTM_Break, time=0.9, value=1}
          {arriveTangent=-15.8223, arriveTangentWeight=0.1667, interpMode=RCIM_Cubic, leaveTangent=-15.8223, tangentMode=RCTM_User, time=1, value=0}
         ]}}}
      CurveIndex=$Particles.Initial.Partition.NormalizedID
      Scale Curve=0.25
     )
     Alpha=&ClampFloat(Float=&Divide_Float(A=&Subtract_Float(A=$Particles.BeamLength, B=250), B=250))
    )
    Scale RGB=&VectorFromFloat(Value=1)
    ScaleA=true
    ScaleRGB=true
   )
  ]
  rendererClasses [&NiagaraRibbonRendererProperties]
  renderers [{rendererClass=&NiagaraRibbonRendererProperties}]
 }
}
summary {
 systemName=NS_ClusterLauncher_TargetingLaser
 emitters [
  {emitterName=MainRibbon, rendererClasses [&NiagaraRibbonRendererProperties]}
  {emitterName=Glow, rendererClasses [&NiagaraSpriteRendererProperties, &NiagaraLightRendererProperties]}
 ]
 userVariables [
  User.BeamEnd_0 {type=NiagaraPosition, description="", defaultValue {}}
  User.CinematicLength {type=NiagaraFloat, defaultValue=250, description=""}
  User.IsCinematic {type=NiagaraBool}
  User.IsLocalPlayer {type=NiagaraBool}
 ]
}
