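# Niagara system for the Cluster Launcher that throws out an ammo pack as a mesh particle, likely as a discarded-pack visual.
# A single burst emitter launches the mesh downward under gravity and shrinks it to zero along a curve over its lifetime.
# Collision bounces it using bounds that follow particle scale, gated by a platform set that excludes the lowest quality levels.
# Random-axis tumbling halves its rate after a collision.
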NiagaraSystem /ClusterLauncher/Assets/FX/NS_ClusterLauncher_ThrowAmmoPack
dependencies {
 usedDataInterfaces [&NiagaraDataInterfaceCurve, &NiagaraDataInterfacePlatformSet]
 usedDynamicInputs []
 usedModules [&SystemState]
}
emitters {NE_AmmoPack {
  emitterUpdateScript [
   &EmitterState()
   &SpawnBurst_Instantaneous(Use Spawn Probability=false, UseLoopCountLimit=false)
  ]
  particleSpawnScript [
   &InitializeParticle(Color Mode=::NewEnumerator0, Mesh Scale Mode=::NewEnumerator3, UsePositionOffset=false, Write Material Random=true)
   &AddVelocity(Velocity=<0 0 -1>, Velocity Speed Scale=250)
  ]
  particleUpdateScript [
   &ParticleState()
   &GravityForce()
   &ScaleMeshSize(Scale Factor=&VectorFromFloat(Value=&FloatFromCurve(FloatCurve=DataInterface() {propertyValues {Curve {keys [
          {interpMode=RCIM_Cubic, tangentMode=RCTM_User, time=0, value=1}
          {arriveTangent=-2.2039, interpMode=RCIM_Cubic, leaveTangent=-2.2039, tangentMode=RCTM_User, time=1, value=0}
         ]}}})))
   &Collision(
    Age Colliding Particles=false
    Collision Enabled=&IsPlatformSetActive(New Platform Set=DataInterface() {propertyValues {Platforms {qualityLevelMask=-4}}})
    KillOnCollision=false
    Mesh Dimensions:Vector3f=&Multiply_Vector(A=$Particles.Scale, B=<10 20 10>)
    Module.Control Rest State Via Surface Normal=false
    Module.EnableMaxCollisionCount=false
    Module.Randomize Collision Normal Vector=false
    Radius Calculation Type=::NewEnumerator1
    Restitution=0.4
   )
   &SolveForcesAndVelocity(Clamp Velocity=false, Limit Acceleration=false)
   &UpdateMeshOrientation(
    Rotation Rate=&MakeCustomFloatFromBool(Bool=$Particles.HasCollided, True Float=0.25, False Float=0.5)
    Rotation Vector=&RandomVector(Vector Scale=1)
   )
  ]
  rendererClasses [&NiagaraMeshRendererProperties]
  renderers [{rendererClass=&NiagaraMeshRendererProperties, rendererIndex=0}]
 }}
summary {
 systemName=NS_ClusterLauncher_ThrowAmmoPack
 emitters [{emitterName=NE_AmmoPack, rendererClasses [&NiagaraMeshRendererProperties]}]
 userVariables []
}
