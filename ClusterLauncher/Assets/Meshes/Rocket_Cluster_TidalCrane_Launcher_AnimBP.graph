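# Animation Blueprint for the Cluster TidalCrane rocket launcher that layers spent-rocket, hose physics and aim-down-sights animation over the input pose.
# When bShowRocketsInLauncher is false, an additive spent pose applies before AnimDynamics simulates the hose bone, with the DisableDynamics curve inversely driving its alpha.
# After DefaultSlot, the Laser_ADS_Reticle_SM state machine adds in, hold and out poses driven by bIsADS, resting on an additive identity pose.
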AnimBlueprint /ClusterLauncher/Assets/Meshes/Rocket_Cluster_TidalCrane_Launcher_AnimBP
target_skeleton=&/Rocket_Cluster_TidalCrane_Launcher_Skeleton
graphs [
 AnimGraph {nodes [
   n1 = Apply Additive(
    Base=Use cached pose 'InPose'()
    Additive="Sequence Player (additive) 'Cluster_Launcher_ADS_Spent_Pose_WEP'"() {bLoopAnimation=false, playRate=0, sequence=/ClusterLauncher/Assets/Animation/Cluster_Launcher_ADS_Spent_Pose_WEP}
   ) {alphaBoolBlend {blendOption=Linear}}
   n2 = Use cached pose 'InPose'()
   n3 = Get bShowRocketsInLauncher() -> (bShowRocketsInLauncher:bool)
   n4 = Anim Dynamics - Bone: dyn_hose_mstr!(
    ComponentPose=Local To Component(LocalPose=Blend Poses by bool(BlendPose_0=@n2, BlendPose_1=@n1, BlendTime_0=0, BlendTime_1=0, bActiveValue=@n3) {blendType=Linear})
    AlphaCurveName:Name=DisableDynamics
    ExternalForce="0,0,0"
   ) -> (Pose) {
    alphaInputType=Curve
    angularDampingOverride=1
    angularSpringConstant=5000
    bAngularSpring=true
    bOverrideAngularDamping=true
    bUseGravityOverride=true
    gravityScale=0.35
    simulationSpace=Actor
    alphaBoolBlend {blendOption=Linear}
    alphaScaleBiasClamp {bMapRange=true, inRange {max=0, min=1}}
    boundBone {boneName=dyn_hose_mstr}
    componentLinearAccScale=<1 1 1>
    componentLinearVelScale=<1 1 1>
    gravityOverride {z=-980}
    physicsBodyDefinitions [{
      boundBone {boneName=dyn_hose_mstr}
      boxExtents=<30 30 30>
      constraintSetup {angularLimitsMax=<45 45 45>, angularLimitsMin=<-45 -45 -45>, angularTarget=<1 0 0>}
      localJointOffset=<6 0 0>
     }]
   }
   n5 = Component To Local(ComponentPose=@n4)
   AnimGraphNode_StateMachine_0 = Laser_ADS_Reticle_SM() {is=title AnimGraphNode_StateMachine}
   Output Pose(Result=Apply Additive(Base=Slot 'DefaultSlot'(Source=@n5), Additive=@AnimGraphNode_StateMachine_0) {alphaBoolBlend {blendOption=Linear}})
   Save cached pose 'InPose'(Pose=Input Pose()) {cacheName=InPose}
  ]}
 Laser_ADS_Reticle_SM {
  entry=Passthrough
  path=AnimGraph.AnimGraphNode_StateMachine_0.Laser_ADS_Reticle_SM
  ids=id
  nodes [
   AnimStateNode_0 = Passthrough() {is=title AnimStateNode}
   AnimStateNode_1 = Into_ADS() {is=title AnimStateNode}
   AnimStateNode_2 = ADS_Loop() {is=title AnimStateNode}
   AnimStateNode_3 = Exit_ADS() {is=title AnimStateNode}
   AnimStateTransitionNode_0 = AnimStateTransitionNode()
   AnimStateTransitionNode_1 = AnimStateTransitionNode() {bAutomaticRuleBasedOnSequencePlayerInState=true, crossfadeDuration=0}
   AnimStateTransitionNode_2 = AnimStateTransitionNode() {crossfadeDuration=0.1}
   AnimStateTransitionNode_3 = AnimStateTransitionNode() {bAutomaticRuleBasedOnSequencePlayerInState=true, crossfadeDuration=0.1}
  ]
  links [
   {
    dst=Exit_ADS
    rule=AnimGraph.AnimGraphNode_StateMachine_0.Laser_ADS_Reticle_SM.AnimStateTransitionNode_2.Transition
    src=ADS_Loop
    via=AnimStateTransitionNode_2
   }
   {
    dst=Passthrough
    rule=AnimGraph.AnimGraphNode_StateMachine_0.Laser_ADS_Reticle_SM.AnimStateTransitionNode_3.Transition
    src=Exit_ADS
    via=AnimStateTransitionNode_3
   }
   {
    dst=ADS_Loop
    rule=AnimGraph.AnimGraphNode_StateMachine_0.Laser_ADS_Reticle_SM.AnimStateTransitionNode_1.Transition
    src=Into_ADS
    via=AnimStateTransitionNode_1
   }
   {
    dst=Into_ADS
    rule=AnimGraph.AnimGraphNode_StateMachine_0.Laser_ADS_Reticle_SM.AnimStateTransitionNode_0.Transition
    src=Passthrough
    via=AnimStateTransitionNode_0
   }
  ]
 }
 Passthrough {
  path=AnimGraph.AnimGraphNode_StateMachine_0.Laser_ADS_Reticle_SM.AnimStateNode_0.Passthrough
  nodes [Output Animation Pose(Result=Additive Identity Pose())]
 }
 Into_ADS {
  path=AnimGraph.AnimGraphNode_StateMachine_0.Laser_ADS_Reticle_SM.AnimStateNode_1.Into_ADS
  nodes [Output Animation Pose(Result="Sequence Player (additive) 'Cluster_Launcher_ADS_In_WEP'"() {bLoopAnimation=false, sequence=/ClusterLauncher/Assets/Animation/Cluster_Launcher_ADS_In_WEP})]
 }
 ADS_Loop {
  path=AnimGraph.AnimGraphNode_StateMachine_0.Laser_ADS_Reticle_SM.AnimStateNode_2.ADS_Loop
  nodes [Output Animation Pose(Result="Sequence Player (additive) 'Cluster_Launcher_ADS_Pose_WEP'"() {bLoopAnimation=false, sequence=/ClusterLauncher/Assets/Animation/Cluster_Launcher_ADS_Pose_WEP})]
 }
 Exit_ADS {
  path=AnimGraph.AnimGraphNode_StateMachine_0.Laser_ADS_Reticle_SM.AnimStateNode_3.Exit_ADS
  nodes [Output Animation Pose(Result="Sequence Player (additive) 'Cluster_Launcher_ADS_Out_WEP'"() {bLoopAnimation=false, sequence=/ClusterLauncher/Assets/Animation/Cluster_Launcher_ADS_Out_WEP})]
 }
 Transition {
  path=AnimGraph.AnimGraphNode_StateMachine_0.Laser_ADS_Reticle_SM.AnimStateTransitionNode_0.Transition
  nodes [Result(bCanEnterTransition=Get bIsADS() -> (bIsADS:bool))]
 }
 Transition {path=AnimGraph.AnimGraphNode_StateMachine_0.Laser_ADS_Reticle_SM.AnimStateTransitionNode_1.Transition, nodes [Result()]}
 Transition {
  path=AnimGraph.AnimGraphNode_StateMachine_0.Laser_ADS_Reticle_SM.AnimStateTransitionNode_2.Transition
  nodes [Result(bCanEnterTransition=NOT Boolean(A=Get bIsADS()))]
 }
 Transition {path=AnimGraph.AnimGraphNode_StateMachine_0.Laser_ADS_Reticle_SM.AnimStateTransitionNode_3.Transition, nodes [Result()]}
 EventGraph {links [], nodes []}
]
variables []
